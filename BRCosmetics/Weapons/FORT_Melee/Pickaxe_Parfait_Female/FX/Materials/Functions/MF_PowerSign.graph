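# MF_PowerSign is a material function from the Pickaxe_Parfait_Female FX folder that computes a sign-preserving power of a scalar.
# It raises the absolute value of Base to Exponent, then multiplies by the sign of Base, so negative inputs stay negative.
# Both inputs can use their preview values as defaults, with Exponent defaulting to squaring.
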MaterialFunction /BRCosmetics/Weapons/FORT_Melee/Pickaxe_Parfait_Female/FX/Materials/Functions/MF_PowerSign
graph {nodes [
  n1 = FunctionInput() {bUsePreviewValueAsDefault=true, inputName=Base, inputType=FunctionInput_Scalar}
  n2 = Abs(None=@n1)
  n3 = FunctionInput() {bUsePreviewValueAsDefault=true, inputName=Exponent, inputType=FunctionInput_Scalar, sortPriority=1, previewValue=<2 0 0 1>}
  FunctionOutput(None=Multiply(A=Power(Base=@n2, Exp=@n3), B=Sign(None=@n1)))
 ]}
