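# Material function for the Parfait pickaxe FX that outputs a hard opacity mask with animated, jagged edges on both sides of the U axis.
# Distance from the U center is added to a scrolling lightning-noise sample scaled by intensity, then floored and inverted to cut the mask.
# Noise U is a per-half constant plus time, so the two edges differ, while noise V combines scaled V terms with time.
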MaterialFunction /BRCosmetics/Weapons/FORT_Melee/Pickaxe_Parfait_Female/FX/Materials/Functions/MF_TrifleParfait_OpacityNoise
graph {nodes [
  Constant() {desc=also random}
  n1 = ConstantBiasScale(None=ComponentMask(None=TextureCoordinate()) {r=true}) {bias=-0.5, scale=2}
  n2 = Multiply(A=Ceil(None=@n1)) {constB=0.5}
  n3 = Time()
  n4 = ComponentMask(None=TextureCoordinate()) {g=true}
  n5 = Add(A=Multiply(A=@n4, B=ScalarParameter() {defaultValue=11, parameterName=Opacity Noise UV Length}), B=@n3) {constB=0, desc=Add random offset}
  n6 = Add(A=Multiply(A=@n4, B=ScalarParameter() {defaultValue=1.618, parameterName=Opacity Noise UV Diagonal}), B=@n3) {constB=0, desc=Add random offset}
  n7 = TextureSample(UVs=AppendVector(A=Add(A=@n2, B=@n3), B=Add(A=@n5, B=@n6))) {
   samplerType=SAMPLERTYPE_LinearGrayscale
   texture=/BRCosmetics/Weapons/FORT_Melee/Pickaxe_Parfait_Female/FX/Textures/T_TrifleParfait_LightningNoise
  }
  n8 = ScalarParameter() {defaultValue=0.666, parameterName=Opacity Noise Intensity}
  n9 = FunctionInput() {bUsePreviewValueAsDefault=true, inputName=Noise Intensity, inputType=FunctionInput_Scalar, previewValue=<1 0 0 1>}
  n10 = Add(A=Abs(None=@n1), B=Multiply(A=Multiply(A=@n7.R, B=@n8), B=@n9))
  FunctionOutput(None=OneMinus(None=Floor(None=@n10))) {outputName=Masked Opacity}
 ]}
