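# Additive unlit Niagara mesh particle material that renders glowing, shimmering cracks for the female Parfait pickaxe.
# Particle color feeds the emissive adjustment function, while the emissive texture's red channel, particle alpha and a time-based sine shimmer form the glow mask.
# Computed mip level relaxes the mask's sharpening exponent with distance; Low and Medium quality use a simple square instead.
# A normalized camera-direction vector, scaled by Camera Offset, supplies an offset toward the viewer.
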Material /BRCosmetics/Weapons/FORT_Melee/Pickaxe_Parfait_Female/FX/Materials/M_Pickaxe_Parfait_Cracks_Glow
graph {
 properties {
  bUsedWithNiagaraMeshParticles=true
  blendMode=BLEND_Additive
  pixelDepthOffsetMode=PDOM_Legacy
  shadingModel=MSM_Unlit
  usedShadingModels=Unlit
 }
 nodes [
  n1 = ParticleColor()
  MaterialFunctionCall(Emissive=@n1.RGB) {materialFunction=&MF_FX_EmissiveAdjustments}
  n2 = CameraPositionWS()
  n3 = WorldPosition()
  Multiply(A=Normalize(VectorInput=Subtract(A=@n2, B=@n3.XYZ)), B=ScalarParameter() {defaultValue=0.5, parameterName=Camera Offset})
  n4 = Time()
  n5 = ScalarParameter() {defaultValue=0.333, parameterName=Shimmer Speed}
  n6 = ConstantBiasScale(None=Sine(None=Multiply(A=@n4, B=@n5)))
  n7 = Saturate(None=ScalarParameter() {defaultValue=0.666, parameterName=Shimmer Intensity})
  n8 = TextureSample() {texture=/BRCosmetics/Weapons/FORT_Melee/Pickaxe_Parfait_Female/Textures/T_Pickaxe_Parfait_Female_E}
  n9 = TextureObject() {texture=&T_Pickaxe_Parfait_Female_E}
  n10 = Divide(A=MaterialFunctionCall!(UVs, Texture Size=TextureProperty(Tex=@n9)) {materialFunction=/Engine/Functions/Engine_MaterialFunctions01/Texturing/ComputeMipLevel}.MipLevel) {constB=5}
  n11 = Power(
   Base=@n8.R
   Exp=LinearInterpolate(A=ScalarParameter() {defaultValue=5, parameterName=Cracks Glow Exponent Close Up}, Alpha=Saturate(None=@n10)) {constA=1}
  )
  n12 = Multiply(A=@n8.R, B=@n8.R)
  Multiply(A=OneMinus(None=Multiply(A=@n6, B=@n7)), B=Multiply(A=@n1.A, B=QualitySwitch(Default=@n11, Low=@n12, High=@n11, Medium=@n12, Epic=@n11)))
 ]
}
