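# Additive unlit Niagara sprite material that draws a soft round glow for the Parfait pickaxe effects.
# Particle color drives emissive through an adjustment function.
# A procedural radial UV gradient, multiplied by a particle-size-aware scene depth fade, is sharpened by an exponent and scaled by particle alpha.
# A camera-facing offset vector scales with particle size and camera distance relative to a limit and blend range.
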Material /BRCosmetics/Weapons/FORT_Melee/Pickaxe_Parfait_Female/FX/Materials/M_Pickaxe_Parfait_Glow
graph {
 properties {
  bUsedWithNiagaraSprites=true
  blendMode=BLEND_Additive
  pixelDepthOffsetMode=PDOM_Legacy
  shadingModel=MSM_Unlit
  translucencyPass=MTP_BeforeDOF
  usedShadingModels=Unlit
 }
 nodes [
  n1 = ParticleColor()
  MaterialFunctionCall(Emissive=@n1.RGB) {materialFunction=&MF_FX_EmissiveAdjustments}
  n2 = MaterialFunctionCall(Float2=ParticleSize()) {materialFunction=&BreakOutFloat2Components}
  n3 = Divide(A=Min(A=@n2.R, B=@n2.R))
  n4 = CameraPositionWS()
  n5 = Subtract(A=@n4, B=WorldPosition().XYZ)
  n6 = Subtract(A=Distance(A=@n4, B=ParticlePositionWS()), B=ScalarParameter() {defaultValue=111, parameterName=Camera Offset Distance Limit})
  n7 = ScalarParameter() {defaultValue=111, parameterName=Camera Offset Distance Blend}
  n8 = ScalarParameter() {defaultValue=1, parameterName=Camera Offset}
  Multiply(A=Multiply(A=@n3, B=Normalize(VectorInput=@n5)), B=Multiply(A=Divide(A=@n6, B=@n7), B=Saturate(None=@n8)))
  n9 = Distance(A=TextureCoordinate(), B=Constant() {r=0.5})
  n10 = OneMinus(None=Saturate(None=Multiply(A=@n9) {constB=2}))
  n11 = Distance(A=SceneDepth().R, B=PixelDepth().R)
  n12 = Saturate(None=Divide(A=Add(A=@n11, B=@n3), B=ScalarParameter() {defaultValue=222, parameterName=Depth Fade Multiplier}))
  Multiply(A=@n1.A, B=Power(Base=Multiply(A=@n10, B=@n12), Exp=ScalarParameter() {defaultValue=3, parameterName=Exponent}))
 ]
}
