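# Material function for the Parfait female pickaxe FX that outputs an animated Brightness oscillation value.
# The V component of the input UV is scaled and paired with a constant U, so lightning noise varies along one axis only while panning equally in both.
# The noise's red channel is scaled by intensity and inverted, dimming brightness from one.
# One vector parameter packs the UV size, pan speed and intensity.
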MaterialFunction /BRCosmetics/Weapons/FORT_Melee/Pickaxe_Parfait_Female/FX/Materials/Functions/MF_TrifleParfait_BrightnessOscillation
graph {nodes [
  n1 = Constant()
  n2 = TextureCoordinate()
  n3 = VectorParameter() {parameterName="BrightnessOscillation UV Size, Speeds, Intensity", defaultValue=<rgba 1 -0.139 0 0.666>}
  n4 = Multiply(A=ComponentMask(None=FunctionInput(Preview=@n2) {bUsePreviewValueAsDefault=true, inputType=FunctionInput_Vector2}) {g=true}, B=@n3.R)
  n5 = TextureSample(UVs=Panner(Coordinate=AppendVector(A=@n1, B=@n4), Speed=AppendVector(A=@n3.G, B=@n3.G)) {speedX=0.222}) {
   samplerSource=SSM_Wrap_WorldGroupSettings
   samplerType=SAMPLERTYPE_LinearGrayscale
   texture=/BRCosmetics/Weapons/FORT_Melee/Pickaxe_Parfait_Female/FX/Textures/T_TrifleParfait_LightningNoise
  }
  FunctionOutput(None=OneMinus(None=Multiply(A=@n5.R, B=@n3.A))) {outputName=Brightness}
  TextureCoordinate()
  TextureCoordinate()
 ]}
